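# Niagara system for the melee swing trail of the CubicVice pickaxe cosmetic.
# A single TrailMain emitter uses AnimTrail modules to spawn and update short-lived particles from skeletal mesh bones or sockets, drawn by two ribbon renderers.
# Initial color passes through a variant selector with time-of-day blending, backed by system-level variant and time-of-day modules.
# A curve fades alpha over each particle's life, with time-of-day scaling set uniform across periods.
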NiagaraSystem /BRCosmetics/Weapons/FORT_Melee/Pickaxe_CubicVice/FX/NS_Pickaxe_CubicVice_Trail
dependencies {
 usedDataInterfaces [&NiagaraDataInterfaceColorCurve, &NiagaraDataInterfaceCurve, &NiagaraDataInterfaceSkeletalMesh]
 usedDynamicInputs []
 usedModules [
  &self:SystemSpawnScript.SystemScriptSource.SystemScriptGraph.NiagaraNodeAssignment_0.SetVariables_#1
  &SystemState
  &FN_VariantManager
  /BRCosmetics/Characters/Player/Female/Medium/Bodies/F_MED_PiperShelf_Pearl/FX/TimeOfDayGlobal
 ]
}
emitters {TrailMain {
  emitterUpdateScript [
   &EmitterState(Life Cycle Mode=::NewEnumerator1, Loop Behavior=::NewEnumerator1, Loop Duration=1, UseLoopDelay=false)
   &FN_SpawnAnimTrail_v2()
  ]
  particleSpawnScript [
   /Game/Effects/Niagara/Modules/AnimTrail/FN_InitializeAnimTrail/FN_InitializeAnimTrailBoneIndecies(Second Bone or Socket Index=1)
   &FN_InitializeAnimTrail(
    Color=&MultiplyLinearColor(
     Linear Color=<rgba 0.5312 0.5312 0.5312 0.5>
     Scale Factor=&MakeVector4_FromLinearColor(LinearColor=&VariantColorSelector(
       Variant Amount=0
       BlendAmount=&MakeCustomFloatFromBool(Bool=$System.bIsVariant0, True Float=0.7, False Float=0.7)
       Use TODCurve?=true
       Original Color=<rgba 0.6667 0.6667 0.6667 0.5>
       Variant 1 Color=&ColorFromCurve(ColorCurve=DataInterface() {propertyValues {
          AlphaCurve {keys [<at 0 0.33 RCIM_Cubic>,<at 1 0>]}
          BlueCurve {keys [<at 0 0.349 RCIM_Cubic>,<at 0.4791 0.0478 RCIM_Cubic>]}
          GreenCurve {keys [<at 0 0.7913 RCIM_Cubic>,<at 0.4791 0.6948 RCIM_Cubic>]}
          RedCurve {keys [<at 0 1 RCIM_Cubic>,<at 0.4791 1 RCIM_Cubic>]}
         }})
       Variant 2 Color=?
      ))
    )
    Lifetime=0.2
    SkeletalMesh=$System.Skeletal Mesh
   ) {module=FN_InitializeAnimTrail001}
  ]
  particleUpdateScript [
   &SolveForcesAndVelocity(Clamp Velocity=false, Limit Acceleration=false)
   &FN_UpdateAnimTrail(SkeletalMesh=$System.Skeletal Mesh)
   &ScaleColor(
    Scale Alpha=&FloatFromCurve(
     FloatCurve=DataInterface() {propertyValues {Curve {keys [{arriveTangent=-2.2741, interpMode=RCIM_Cubic, leaveTangent=-2.2741, tangentMode=RCTM_User, time=0, value=1}, <at 0.6 0 RCIM_Cubic>]}}}
     Scale Curve=&FNTimeOfDayScalar(Night Scale=1, Morning Scale=1, Evening Scale=1)
    )
    ScaleA=true
    ScaleRGB=true
   )
  ]
  rendererClasses [&NiagaraRibbonRendererProperties]
  renderers [
   {rendererClass=&NiagaraRibbonRendererProperties}
   {rendererClass=&NiagaraRibbonRendererProperties, rendererIndex=1}
  ]
 }}
summary {
 systemName=NS_Pickaxe_CubicVice_Trail
 emitters [{emitterName=TrailMain, rendererClasses [&NiagaraRibbonRendererProperties]}]
 userVariables [
  User.bIsFrontEnd {type=NiagaraBool}
  User.bIsFrontEndPreview {type=NiagaraBool}
  User.DebugVelocity {type=NiagaraFloat, defaultValue=0, description=""}
  User.RibbonIDOffset {type=NiagaraInt32}
  User.Variant {type=NiagaraFloat, defaultValue=0, description=""}
 ]
}
