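# Physics asset for the CubicVice axe, a two-handed melee pickaxe cosmetic.
# It defines a single body on the Melee_TwoHanded_Handle bone, approximated by one thin capsule offset and rotated from the bone origin.
# It has no constraints, so there is no articulated or jointed simulation between parts.
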PhysicsAsset /BRCosmetics/Weapons/FORT_Melee/Pickaxe_CubicVice/Meshes/CubicVice_Axe_PhysicsAsset
bodies {Melee_TwoHanded_Handle [{shapeType=Capsule, length=21.5105, radius=2.3314, center=<0.0746 -3.1561 26.3363>, rotation=<rot -0.0014 -90 0.5445>}]}
constraints []
